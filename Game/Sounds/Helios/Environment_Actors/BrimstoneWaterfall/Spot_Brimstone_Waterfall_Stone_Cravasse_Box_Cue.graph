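# Looping ambient water sound cue for a spot brimstone waterfall in a stone crevasse, routed to the Fort_Ambient_Water sound class.
# A mixer blends three looping waves: the crevasse interior loop at reduced level, a boosted Cerberus large waterfall layer and a rumble loop.
# Overridden attenuation uses a box shape with falloff and listener focus, quieting the sound when out of focus.
# Overridden concurrency limits simultaneous instances to two, stopping the quietest.
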SoundCue /Game/Sounds/Helios/Environment_Actors/BrimstoneWaterfall/Spot_Brimstone_Waterfall_Stone_Cravasse_Box_Cue
root=Mixer_0
nodes [
 {id=Mixer_0, children [WavePlayer_0,WavePlayer_2,WavePlayer_1], properties {inputVolume [0.5,2,1]}}
 {
  id=WavePlayer_0
  properties {bLooping=true, soundWaveAssetPtr=&/Spot_Brimstone_Waterfall_Stone_Cravasse_Int_Loop}
 }
 {
  id=WavePlayer_1
  properties {bLooping=true, soundWaveAssetPtr=&/Spot_Brimstone_Waterfall_Rumble_Loop}
 }
 {
  id=WavePlayer_2
  properties {bLooping=true, soundWaveAssetPtr=&/Spot_Brimstone_Waterfall_Cerberus_Large_01}
 }
]
properties {
 bOverrideAttenuation=true
 bOverrideConcurrency=true
 maxDistance=1200
 priority=10
 soundClassObject=/Game/Sounds/FortSoundClassesAndModes/Fort_Ambient_Water
 volumeMultiplier=0.5
 attenuationOverrides {attenuationShape=Box, bEnableListenerFocus=true, falloffDistance=750, nonFocusVolumeAttenuation=0.5, attenuationShapeExtents=<450 450 400>}
 concurrencyOverrides {maxCount=2, resolutionRule=StopQuietest}
}
