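# Looping ambient water cue for a brimstone waterfall stone crevasse detail spot in the Helios environment.
# The root mixer layers large waterfall and bass rumble loops with a sub-mix of cave river and close-focus crevasse loops.
# A fourth branch starts the crevasse interior loop after a short random delay, under a fade-in volume envelope.
# Overrides give it box-shaped attenuation, a distance-scaled send to distant reverb, and low concurrency that stops the quietest instance.
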SoundCue /Game/Sounds/Helios/Environment_Actors/BrimstoneWaterfall/Spot_Brimstone_Waterfall_Stone_Cravasse_Box_Detail_Cue
root=Mixer_0
nodes [
 {id=Delay_0, children [WavePlayer_2], properties {delayMax=0.4, delayMin=0.2}}
 {
  id=Enveloper_1
  children [Delay_0]
  properties {pitchCurve {editorCurveData {keys [{time=0}]}}, volumeCurve {editorCurveData {keys [<at 0 0>,<at 1 1>]}}}
 }
 {id=Mixer_0, children [Mixer_1,Enveloper_1,WavePlayer_0,WavePlayer_5], properties {inputVolume [2,2,2,2]}}
 {id=Mixer_1, children [WavePlayer_4,WavePlayer_1], properties {inputVolume [1,1]}}
 {
  id=WavePlayer_0
  properties {bLooping=true, soundWaveAssetPtr=&/Spot_Brimstone_Waterfall_Cerberus_Large_01}
 }
 {
  id=WavePlayer_1
  properties {
   bLooping=true
   soundWaveAssetPtr=&/Spot_Brimstone_Waterfall_Stone_Cravasse_Close_Focus_Int_Loop
  }
 }
 {
  id=WavePlayer_2
  properties {bLooping=true, soundWaveAssetPtr=&/Spot_Brimstone_Waterfall_Stone_Cravasse_Int_Loop}
 }
 {id=WavePlayer_4, properties {bLooping=true, soundWaveAssetPtr=/Game/Sounds/Fort_Ambient_Looping/Water/Spot_Water_Cave_River_Flowing}}
 {
  id=WavePlayer_5
  properties {bLooping=true, soundWaveAssetPtr=&/Spot_Brimstone_Waterfall_Rumble_Bass_Loop}
 }
]
properties {
 bOverrideAttenuation=true
 bOverrideConcurrency=true
 maxDistance=1037
 priority=10
 soundClassObject=/Game/Sounds/FortSoundClassesAndModes/Fort_Ambient_Water
 volumeMultiplier=0.2
 attenuationOverrides {
  attenuationShape=Box
  bEnableSubmixSends=true
  falloffDistance=187
  nonFocusVolumeAttenuation=0.5
  occlusionLowPassFilterFrequency=1000
  occlusionTraceChannel=ECC_GameTraceChannel11
  occlusionVolumeAttenuation=0.75
  attenuationShapeExtents=<250 348 850>
  submixSendSettings [{
    bEnableHPFCutoff=false
    bEnableLPFCutoff=false
    disableManualSendClamp=false
    hPFCutoff=20
    lPFCutoff=20000
    maxSendDistance=1000
    maxSendLevel=1
    minSendDistance=400
    minSendLevel=0.2
    sendLevel=1
    sendLevelControlMethod=Linear
    soundSubmix=&Fortnite_Reverb_Distant
   }]
 }
 concurrencyOverrides {maxCount=2, resolutionRule=StopQuietest}
}
