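# Looping splashy impact sound cue for the Brimstone Waterfall, in the water sound class with borrowed pipe-overflow attenuation and a two-voice StopQuietest concurrency limit.
# A mixer blends splash, thick-liquid, concrete-splashing, waterfall-bottom and faint whispy-magic layers, with only some wave players set to loop.
# The bottom layer gets a short random delay and a volume envelope fade-in.
# A modulator randomizes pitch slightly, and all quality levels share that same branch.
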SoundCue /Game/Sounds/Helios/Environment_Actors/BrimstoneWaterfall/Spot_Water_Brimestone_Waterfall_Splashy_Impact_Loop_Cue
root=QualityLevel_0
nodes [
 {id=Delay_0, children [WavePlayer_3], properties {delayMax=0.5, delayMin=0.2}}
 {
  id=Enveloper_0
  children [Delay_0]
  properties {pitchCurve {editorCurveData {keys [{time=0}]}}, volumeCurve {editorCurveData {keys [<at 0 0>,<at 1 1>]}}}
 }
 {id=Mixer_0, children [WavePlayer_1,WavePlayer_0,WavePlayer_2,Enveloper_0,WavePlayer_4], properties {inputVolume [2,0.5,0.75,1,0.1]}}
 {id=Modulator_1, children [Mixer_0], properties {pitchMax=1.1, pitchMin=0.9, volumeMax=1, volumeMin=1}}
 {id=QualityLevel_0, children [Modulator_1,Modulator_1,Modulator_1,Modulator_1]}
 {
  id=WavePlayer_0
  properties {soundWaveAssetPtr=&/Spot_Brimstone_Waterfall_Thick_Liquid_Impact_Loop}
 }
 {
  id=WavePlayer_1
  properties {bLooping=true, soundWaveAssetPtr=/Game/Sounds/Asteria/Environment_Actors/WaterPipeOverflow/Spot_Water_PiperOverflow_SplashImpact_Loop_02}
 }
 {
  id=WavePlayer_2
  properties {soundWaveAssetPtr=&/Spot_Water_Splashing_Concrete_Int_Loop_01}
 }
 {
  id=WavePlayer_3
  properties {bLooping=true, soundWaveAssetPtr=&/Spot_Brimstone_Waterfall_Bottom_Loop}
 }
 {
  id=WavePlayer_4
  properties {bLooping=true, soundWaveAssetPtr=/Game/Sounds/Helios/Environment_Actors/BrimstoneCerberusHead/Spot_Cerberus_Head_Whispy_Magic_Fast_Loop}
 }
]
properties {
 attenuationSettings=/Game/Sounds/Asteria/Environment_Actors/WaterPipeOverflow/Spot_Water_PipeOverflow_Splash_Impact_Heavy_Occ_Att
 bOverrideConcurrency=true
 maxDistance=1250
 priority=10
 soundClassObject=/Game/Sounds/FortSoundClassesAndModes/Fort_Water
 volumeMultiplier=0.2
 concurrencyOverrides {maxCount=2, resolutionRule=StopQuietest}
}
